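; Looping gameplay cue that drives tow hook audio while the projectile is attached to a vehicle.
; It plays 1P/3P attach, detach and reeling sounds, ducking engine audio when the local view is the instigator or a passenger.
; A repeating timer compares instigator-to-vehicle distance against the farthest point, using a threshold to trigger reel-in/out changes.
; Skidding tags trigger loop parameters, and Sound Library movement events add footstep foley when not skidding.
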(event EventOnLoopingStart(EitherVFX-preferthisversion)
  (params (MyTarget :type Actor%20Object%20Reference) (Parameters :type Gameplay%20Cue%20Parameters%20Structure%20%28by%20ref%29) (ParticleComponents :type Array%20of%20FXSystem%20Component%20Object%20References) (AudioComponents :type Array%20of%20Audio%20Component%20Object%20References))
  (bind (_normalizedmagnitude _rawmagnitude _effectcontext _matchedtagname _originaltag _aggregatedsourcetags _aggregatedtargettags _location _normal _instigator _effectcauser) (Ability|GameplayCue|BreakGameplayCueParameters Parameters))
  (|CallParentFunction MyTarget Parameters ParticleComponents AudioComponents)
  (Utilities|FlowControl|Sequence
    (:then_0
      ;[0.06418 0.1354 0.07433] Cache References
      (Variables|Default|SetAttached%20Projectile%20Actor MyTarget))
    (:then_1
      (Utilities|FlowControl|Sequence
        (:then_0
          (bind _asfort_player_pawn (Utilities|Casting|CastToFortPlayerPawn _instigator))
          (bind _output_get (Variables|Default|SetInstigator%20Pawn _asfort_player_pawn))
          (EventDispatchers|BindEventtoOnSoundLibraryPlayEvent (Actor|GetComponentbyClass _output_get "/Script/SoundLibrary.SoundLibraryComponent") (EventDispatchers|CreateEvent :!event Custom|OnSoundLibraryPlayEvent)))
        (:then_1
          (Utilities|FlowControl|Sequence
            (:then_0
              (bind _asfort_athena_vehicle (Utilities|Casting|CastToFortAthenaVehicle _effectcauser))
              (bind _output_get_2 (Variables|Default|SetAttached%20Vehicle _asfort_athena_vehicle))
              (bind _instigator_pawn (Variables|Default|GetInstigator%20Pawn)
                (:then
                  (Variables|Default|SetStarting%20Distance (Transformation|GetDistanceTo _instigator_pawn _output_get_2)))
                (:else)))
            (:then_1
              (Utilities|FlowControl|Sequence
                (:then_0
                  ;[0.2448 0.1973 0.09435] Play attach sound
                  (CallFunction|PlayProjectileOneShot1P/3P (Variables|Sounds|GetProjectile%20Attach%201P%20Sound) (Variables|Sounds|GetProjectile%20Attach%203P%20Sound) (Variables|/Script/Engine.Actor|GetRootComponent (Variables|Default|GetAttached%20Projectile%20Actor)))
                  ;[0.1275 0 0.2708] Play reeling loop sound
                  (bind _audio_component (Default|PlayReelingSound))
                  (Variables|Default|SetReeling%20Audio _audio_component)
                  ;[0.2031 0.07999 0] Update reeling loop sounds — checks if we are reeling in or reeling out
                  (Utilities|Time|SetTimerbyEvent (delegate Custom|Update%20Reeling%20Audio) 0.1 true)
                  (CallFunction|UpdateReelingAudio))
                (:then_1
                  (Utilities|FlowControl|Sequence
                    (:then_0
                      ;[0 0.1487 0.2292] Play reeling only while player is skidding.
                      (bind _instigator_pawn_2 (Variables|Default|GetInstigator%20Pawn)
                        (:then
                          (Ability|Tasks|WaitGameplayTagAddToActor _instigator_pawn_2 (Variables|Default|GetSkidding%20Tag)
                            (:then
                              (Ability|Tasks|WaitGameplayTagRemoveFromActor (Variables|Default|GetInstigator%20Pawn) (Variables|Default|GetSkidding%20Tag)
                                (:Removed
                                  (bind _reeling_audio (Variables|Default|GetReeling%20Audio)
                                    (:then
                                      (Audio|Parameter|ExecuteTriggerParameter _reeling_audio "Events.OnSkiddingEnd"))
                                    (:else)))))
                            (:Added
                              (bind _reeling_audio_2 (Variables|Default|GetReeling%20Audio)
                                (:then
                                  (Audio|Parameter|ExecuteTriggerParameter _reeling_audio_2 "Events.OnSkiddingStart"))
                                (:else)))))
                        (:else)))))))))))))

(event Custom|Update%20Reeling%20Audio
  ;[0.2031 0.07999 0] Update reeling loop sounds — checks if we are reeling in or reeling out
  (CallFunction|UpdateReelingState))

(event Custom|Duck%20Vehicle%20Engine%20Audio
  ;[0.2642 0.04995 0.2708] Engine audio ducking for attach / detach sounds
  (Utilities|FlowControl|Delay 0)
  (Audio|Modulation|ActivateControlBusMix 0 (Variables|Default|GetControlBusMix))
  (Utilities|FlowControl|Delay 0 0.1)
  (Audio|Modulation|DeactivateControlBusMix 0 (Variables|Default|GetControlBusMix)))

(event Custom|OnSoundLibraryPlayEvent
  (params (Results :type Sound%20Library%20Play%20Result%20Structure%20%28by%20ref%29) (Context :type Sound%20Library%20Context%20Object%20Reference))
  ;[0.1406 0.05493 0.05493] Footstep foley layer while attached to the vehicle but not being dragged.
  (bind _instigator_pawn (Variables|Default|GetInstigator%20Pawn)
    (:then
      (if (and (GameplayTags|MatchesAnyTags (Utilities|Struct|BreakSoundLibraryPlayResult Results) (Variables|Default|GetMovementTags) true) (not (GameplayTags|HasTag (Variables|/Script/FortniteGame.FortPawn|GetGameplayTags _instigator_pawn) (Variables|Default|GetSkidding%20Tag))))
        (bind _instigator_pawn_2 (Variables|Default|GetInstigator%20Pawn)
          (:then
            (Audio|SpawnSoundAttached (Variables|Sounds|GetWalk%20Sound) (Variables|/Script/Engine.Actor|GetRootComponent _instigator_pawn_2)))
          (:else))))
    (:else)))

(event EventOnRemoval(EitherVFX-preferthisversion)
  (params (MyTarget :type Actor%20Object%20Reference) (Parameters :type Gameplay%20Cue%20Parameters%20Structure%20%28by%20ref%29) (ParticleComponents :type Array%20of%20FXSystem%20Component%20Object%20References) (AudioComponents :type Array%20of%20Audio%20Component%20Object%20References) (BurstCameraShakeInstance :type Camera%20Shake%20Base%20Object%20Reference) (BurstDecalInstance :type Decal%20Actor%20Object%20Reference))
  ;[0.05729 0.05729 0.05729] On Removal
  (|CallParentFunction MyTarget Parameters ParticleComponents AudioComponents BurstCameraShakeInstance BurstDecalInstance)
  (CallFunction|PlayProjectileOneShot1P/3P (Variables|Sounds|GetProjectile%20Detach%201P%20Sound) (Variables|Sounds|GetProjectile%20Detach%203P%20Sound) (Variables|/Script/Engine.Actor|GetRootComponent (Variables|Default|GetAttached%20Vehicle))) ; Attach sound to the vehicle to prevent sound getting left behind when projectile is killed
  (Audio|Components|Audio|FadeOut (Variables|Default|GetReeling%20Audio) 0.2 0.0 "Linear" :!class /Script/Engine.AudioComponent)
  (bind _instigator_pawn (Variables|Default|GetInstigator%20Pawn)
    (:then
      (EventDispatchers|UnbindEventfromOnSoundLibraryPlayEvent (Actor|GetComponentbyClass _instigator_pawn "/Script/SoundLibrary.SoundLibraryComponent") (EventDispatchers|CreateEvent :!event Custom|OnSoundLibraryPlayEvent)))
    (:else)))

(fn PlayProjectileOneShot1P/3P
  (params (1P%20Sound :type Sound%20Base%20Object%20Reference) (3P%20Sound :type Sound%20Base%20Object%20Reference) (Attach%20To :type Scene%20Component%20Object%20Reference))
  (Utilities|FlowControl|Sequence
    (:then_0
      ;[0.09208 0.04834 0.09375] Play 1P or 3P sound
      (bind _instigator_pawn (Variables|Default|GetInstigator%20Pawn)
        (:then
          (bind _returnvalue (Audio|SpawnSoundAttached (select (Pawn|IsLocallyViewed _instigator_pawn) (Variables|Get1P%20Sound) (Variables|Get3P%20Sound)) (Variables|GetAttach%20To)))
          (LocalVariables|SetLocal%20Audio%20Component _returnvalue))
        (:else)))
    (:then_1
      (Utilities|FlowControl|Sequence
        (:then_0
          ;[0.007862 0.1138 0.1354] Duck the engine audio volume if listener is the instigator or inside of the hit vehicle
          (if (CallFunction|ShouldDuckEngineAudio)
            (CallFunction|DuckVehicleEngineAudio)))
        (:then_1
          (Utilities|FlowControl|Sequence
            (:then_0
              (return (LocalVariables|GetLocal%20Audio%20Component)))))))))

(fn PlayReelingSound
  (bind _instigator_pawn (Variables|Default|GetInstigator%20Pawn)
    (:then
      (bind _attached_projectile_actor (Variables|Default|GetAttached%20Projectile%20Actor)
        (:then
          (bind _returnvalue (Audio|SpawnSoundAttached (select (Pawn|IsLocallyViewed _instigator_pawn) (Variables|Sounds|GetReeling%20Loop%201P) (Variables|Sounds|GetReeling%20Loop%203P)) (Variables|/Script/Engine.Actor|GetRootComponent _attached_projectile_actor)))
          (return _returnvalue))
        (:else)))
    (:else)))

(fn SetReelingState
  (params (Reference%20Value :type Float%20%28double-precision%29))
  (bind _reeling_state (Variables|Default|GetReeling%20State))
  (bind _returnvalue (Utilities|Enum|Equal(Enum) _reeling_state "NewEnumerator1"))
  (bind _returnvalue_2 (Utilities|String|SelectName "Events.OnReelIn" "Events.OnReelOut" _returnvalue))
  (if (> (Variables|Default|GetCurrent%20Distance) (Variables|GetReference%20Value))
    (Variables|Default|SetReeling%20State "NewEnumerator2")
    (else
      (Variables|Default|SetReeling%20State "NewEnumerator1")))
  (Audio|Parameter|ExecuteTriggerParameter (Variables|Default|GetReeling%20Audio) _returnvalue_2)
  ; Print Debug State
  (if true
    (Development|PrintText 0 (Utilities|Text|ToText(Name) _returnvalue_2) true false "(R=0.701580,G=0.000000,B=1.000000,A=1.000000)" 5.0 "ReelState")))

(fn ShouldDuckEngineAudio
  (bind _instigator_pawn (Variables|Default|GetInstigator%20Pawn)
    (:then
      (if (Pawn|IsLocallyViewed _instigator_pawn)
        (return true)
        (else
          (bind _array_element (Utilities|Array|ForEachLoopwithBreak (Vehicle|GetAllPassengers (Variables|Default|GetAttached%20Vehicle)))
            (:LoopBody
              (if (Pawn|IsLocallyViewed _array_element)
                (LocalVariables|SetLocal%20Is%20Passenger true))
              (LocalVariables|SetLocal%20Is%20Passenger false))
            (:Completed
              _))
          (return (LocalVariables|GetLocal%20Is%20Passenger)))))
    (:else)))

(fn UpdateReelingState
  (Variables|Default|GetReeling%20Audio
    (:then
      (Utilities|FlowControl|Sequence
        (:then_0
          ;[0.08913 0.1035 0.3385] Calculate current distance
          (bind _attached_vehicle (Variables|Default|GetAttached%20Vehicle)
            (:then
              (Variables|Default|SetCurrent%20Distance (Transformation|GetDistanceTo (Variables|Default|GetInstigator%20Pawn) _attached_vehicle)))
            (:else)))
        (:then_1
          (Utilities|FlowControl|Sequence
            (:then_0
              ;[0.2656 0.1463 0.03019] Set reeling state based on starting distance
              (if (Utilities|Enum|Equal(Enum) (Variables|Default|GetReeling%20State) "NewEnumerator0")
                (CallFunction|SetReelingState (Variables|Default|GetStarting%20Distance))))
            (:then_1
              (Utilities|FlowControl|Sequence
                (:then_0
                  ;[0.04857 0.1927 0.1618] Keep track of the farthest distance toward or away from the vehicle, so we can set a threshold at which state will change.
                  (switch Utilities|FlowControl|Switch|SwitchonTowHookReelStates (Variables|Default|GetReeling%20State)
                    (:NewEnumerator1
                      (Variables|Default|SetFarthest%20Distance (Math|Float|SelectFloat (Variables|Default|GetFarthest%20Distance) (Variables|Default|GetCurrent%20Distance) (> (Variables|Default|GetCurrent%20Distance) (Variables|Default|GetFarthest%20Distance)))))
                    (:NewEnumerator2
                      (Variables|Default|SetFarthest%20Distance (Math|Float|SelectFloat (Variables|Default|GetCurrent%20Distance) (Variables|Default|GetFarthest%20Distance) (> (Variables|Default|GetCurrent%20Distance) (Variables|Default|GetFarthest%20Distance))))))
                  ; Print Debug Distances
                  (if false
                    (Development|PrintText 0 (Utilities|Text|FormatText "NSLOCTEXT(\"[01672D3C93D9DCE0F8F8ACBC84D86DB3]\", \"A12276ED4C58F75E1A8073BCB6F449EF\", \"Distance: {distance}\\r\\nFarthest Distance: {farthest}\\r\\nThreshold: {thresh}\")" (Variables|Default|GetCurrent%20Distance) (Variables|Default|GetFarthest%20Distance) (Math|Float|SelectFloat (- (Variables|Default|GetFarthest%20Distance) (Variables|GetSignificance%20Threshold)) (+ (Variables|Default|GetFarthest%20Distance) (Variables|GetSignificance%20Threshold)) (Utilities|Enum|Equal(Enum) (Variables|Default|GetReeling%20State) "NewEnumerator2"))) true true "(R=1.000000,G=0.353337,B=0.000000,A=1.000000)" 2.0 "ReelDistance")))
                (:then_1
                  (Utilities|FlowControl|Sequence
                    (:then_0
                      ;[0.2396 0.1136 0.1513] Change the reeling state if the current distance exceeds the significance threshold in the opposite direction of the reeling.
                      (if (> (Math|Float|Absolute(Float) (- (Variables|Default|GetFarthest%20Distance) (Variables|Default|GetCurrent%20Distance))) (LocalVariables|GetSignificance%20Threshold))
                        (CallFunction|SetReelingState (Variables|Default|GetFarthest%20Distance))))
                    (:then_1
                      (Utilities|FlowControl|Sequence
                        (:then_0
                          (Audio|Parameter|SetFloatParameter (Variables|Default|GetReeling%20Audio) "InstigatorDistance" (Variables|Default|GetCurrent%20Distance) :!class /Script/AudioExtensions.AudioParameterControllerInterface))))))))))))
    (:else)))

(fn ConstructionScript
  (|CallParentFunction))
