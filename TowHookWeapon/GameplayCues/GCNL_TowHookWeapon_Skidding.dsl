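; Looping gameplay cue that drives skidding Niagara FX and audio on a pawn tethered by the Tow Hook Weapon.
; On start it disables default player dust and binds grounded, jump and surface-change events.
; These feed Niagara parameters and pick data-table surface sounds, using "Air" on jumps, for audio parameters.
; A jump-pressed async wait re-arms on landing; water and Niagara-deactivation handlers are empty, and removal leaves dust disabled.
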(event EventOnLoopingStart(Niagara)
  (params (MyTarget :type Actor%20Object%20Reference) (Parameters :type Gameplay%20Cue%20Parameters%20Structure%20%28by%20ref%29) (ParticleComponents :type Array%20of%20Niagara%20Particle%20System%20Component%20Object%20References) (AudioComponents :type Array%20of%20Audio%20Component%20Object%20References))
  (bind _as_fort_pawn (Variables|Default|GetAs%20Fort%20Pawn))
  (bind _returnvalue (Actor|GetComponentbyClass _as_fort_pawn "/Script/TowHookWeaponRuntime.FortPawnComponent_TetheredMovement"))
  (|CallParentFunction MyTarget Parameters ParticleComponents AudioComponents)
  (bind _asfort_pawn (Utilities|Casting|CastToFortPawn MyTarget))
  (Variables|Default|SetAs%20Fort%20Pawn _asfort_pawn)
  (Variables|Default|SetSkidding%20FX (Utilities|Array|Get(acopy) ParticleComponents))
  (Variables|Default|SetSkidding%20Audio (Utilities|Array|Get(acopy) AudioComponents))
  (Utilities|FlowControl|Sequence
    (:then_0
      (Components|Activation|BindEventtoOnComponentDeactivated (Variables|Default|GetSkidding%20FX) (EventDispatchers|CreateEvent :!event Custom|OnNiagaraDeactivatedEvent))
      (Utilities|IsValid _returnvalue
        (:Is%20Valid
          (EventDispatchers|BindEventtoOnGroundedStateChanged _returnvalue (EventDispatchers|CreateEvent :!event Custom|OnGroundedStateChangedCustomEvent))
          (EventDispatchers|BindEventtoOnWaterStateChanged _returnvalue (EventDispatchers|CreateEvent :!event Custom|OnWaterStateChangeCustomEvent))
          (CallFunction|OnJumpAsyncWait))))
    (:then_1
      (Pawn|BindEventtoOnPawnJumped (Variables|Default|GetAs%20Fort%20Pawn) (EventDispatchers|CreateEvent :!event Custom|OnJump))
      (EventDispatchers|BindEventtoOnSurfaceTypeChanged (Footsteps|GetSurfaceTrackingComponent (Variables|Default|GetAs%20Fort%20Pawn)) (EventDispatchers|CreateEvent :!event Custom|OnSurfaceChanged)))
    (:then_2
      ;[0.566 0 1] Turn off player walk & run dust FX on tether start
      (CallFunction|EnableDefaultPlayerDustFX false)
      (Audio|UpdateSurfaceSoundParameters (CallFunction|GetCurrentSurface)))))

(event EventOnRemoval(EitherVFX-preferthisversion)
  (params (MyTarget :type Actor%20Object%20Reference) (Parameters :type Gameplay%20Cue%20Parameters%20Structure%20%28by%20ref%29) (ParticleComponents :type Array%20of%20FXSystem%20Component%20Object%20References) (AudioComponents :type Array%20of%20Audio%20Component%20Object%20References) (BurstCameraShakeInstance :type Camera%20Shake%20Base%20Object%20Reference) (BurstDecalInstance :type Decal%20Actor%20Object%20Reference))
  (|CallParentFunction MyTarget Parameters ParticleComponents AudioComponents BurstCameraShakeInstance BurstDecalInstance)
  (bind _aspawn (Utilities|Casting|CastToPawn MyTarget))
  (Utilities|IsValid (Actor|GetComponentbyClass _aspawn "/Script/TowHookWeaponRuntime.FortPawnComponent_TetheredMovement")
    (:Is%20Valid
      (EventDispatchers|UnbindEventfromOnGroundedStateChanged (Actor|GetComponentbyClass _aspawn "/Script/TowHookWeaponRuntime.FortPawnComponent_TetheredMovement") (EventDispatchers|CreateEvent :!event Custom|OnGroundedStateChangedCustomEvent))
      (EventDispatchers|UnbindEventfromOnWaterStateChanged (Actor|GetComponentbyClass _aspawn "/Script/TowHookWeaponRuntime.FortPawnComponent_TetheredMovement") (EventDispatchers|CreateEvent :!event Custom|OnWaterStateChangeCustomEvent))
      (Pawn|UnbindEventfromOnPawnJumped (Variables|Default|GetAs%20Fort%20Pawn) (EventDispatchers|CreateEvent :!event Custom|OnJump)))))

(event Custom|OnGroundedStateChangedCustomEvent
  (params (bGrounded :type Boolean))
  (bind _returnvalue (not bGrounded))
  (Niagara|SetNiagaraVariable(Bool) (Variables|Default|GetSkidding%20FX) "bIsGrounded" bGrounded)
  (if bGrounded
    ;[0.566 0 1] Turn on/off player walk & run dust FX on grounded state change
    (CallFunction|EnableDefaultPlayerDustFX _returnvalue)
    (Niagara|SetNiagaraVariable(Bool) (Variables|Default|GetSkidding%20FX) "blsOnJump" _returnvalue)
    (CallFunction|OnJumpAsyncWait)
    (Audio|UpdateSurfaceSoundParameters (CallFunction|GetCurrentSurface))))

(event Custom|OnWaterStateChangeCustomEvent
  (params (bInWater :type Boolean)))

(event Custom|OnNiagaraDeactivatedEvent
  (params (Component :type Actor%20Component%20Object%20Reference)))

(event Custom|OnJump
  (Audio|UpdateSurfaceSoundParameters "SurfaceType_Default" true)
  (Niagara|SetNiagaraVariable(Bool) (Variables|Default|GetSkidding%20FX) "blsOnJump" true))

(event Custom|OnSurfaceChanged
  (params (SurfaceActor :type Actor%20Object%20Reference) (CurrentSurfaceType :type EPhysicalSurface%20Enum :default "SurfaceType_Default") (PrevSurfaceType :type EPhysicalSurface%20Enum :default "SurfaceType_Default"))
  (bind _output_get (Variables|Default|SetCurrent%20Surface%20Type CurrentSurfaceType))
  (Audio|UpdateSurfaceSoundParameters _output_get)
  (Effects|Components|ParticleSystem|SetIntParameter (Variables|Default|GetSkidding%20FX) "PhysicalSurface" (Math|Conversions|ToInteger(Byte) _output_get)))

(event Custom|OnJumpAsyncWait
  (bind _on_jump_event_async_action (Variables|Default|GetOn%20Jump%20Event%20Async%20Action))
  ;[1 1 1] Jump Event
  (Utilities|IsValid _on_jump_event_async_action
    (:Is%20Valid
      (Ability|Async|EndAction _on_jump_event_async_action))
    (:Is%20Not%20Valid
      _))
  (bind _asynctaskproxy (Ability|Async|WaitGameplayEventToActor (Variables|Default|GetAs%20Fort%20Pawn) "(TagName=\"Event.Jump.Pressed\")" true)
    (:then
      (Variables|Default|SetOn%20Jump%20Event%20Async%20Action _asynctaskproxy))
    (:EventReceived
      (CallFunction|OnJump)
      (Ability|Async|EndAction (Variables|Default|GetOn%20Jump%20Event%20Async%20Action)))))

(fn EnableDefaultPlayerDustFX
  (params (EnableDefaultDustFX :type Boolean))
  (bind _asplayer_pawn_athena (Utilities|Casting|CastToPlayerPawn_Athena (Variables|Default|GetAs%20Fort%20Pawn)))
  (Variables|/Script/FortniteGame.FortPlayerPawn|SetbFXPlayDustOnMovement EnableDefaultDustFX _asplayer_pawn_athena))

(fn GetCurrentSurface
  (return (Class|FortSurfaceTrackingComponent|GetCurrentSurfaceType (Footsteps|GetSurfaceTrackingComponent (Variables|Default|GetAs%20Fort%20Pawn)))))

(fn GetCurrentSurfaceName
  (params (Physical%20Surface%20Type :type EPhysicalSurface%20Enum :default "SurfaceType_Default"))
  (return (select (or (Variables|/Script/Engine.Character|GetbWasJumping (Variables|Default|GetAs%20Fort%20Pawn)) (AI|Components|NavMovement|IsFalling (Variables|/Script/Engine.Character|GetCharacterMovement (Variables|Default|GetAs%20Fort%20Pawn)))) "Air" (Utilities|Select :Index Physical%20Surface%20Type :SurfaceType_Default "Stone" :SurfaceType1 "Wood" :SurfaceType2 "Stone" :SurfaceType3 "Metal" :SurfaceType4 "None" :SurfaceType5 "None" :SurfaceType6 "None" :SurfaceType7 "None" :SurfaceType8 "Water" :SurfaceType9 "Grass" :SurfaceType10 "Dirt" :SurfaceType11 "None" :SurfaceType12 "None" :SurfaceType13 "None" :SurfaceType14 "Wood" :SurfaceType15 "Stone" :SurfaceType16 "Metal" :SurfaceType17 "None" :SurfaceType18 "Stone" :SurfaceType19 "None" :SurfaceType20 "None" :SurfaceType21 "None" :SurfaceType22 "None" :SurfaceType23 "Snow" :SurfaceType24 "Ice" :SurfaceType25 "Dirt" :SurfaceType26 "Sand" :SurfaceType27 "None" :SurfaceType28 "Stone" :SurfaceType29 "Stone" :SurfaceType30 "Mud" :SurfaceType31 "WetConcrete" :SurfaceType32 "Grass" :SurfaceType33 "Stone" :SurfaceType34 "None"))))

(fn GetPhysicalSurfaceName
  (params (Index :type EPhysicalSurface%20Enum :default "SurfaceType_Default"))
  (return (Utilities|Select :Index Index :SurfaceType_Default "Dirt" :SurfaceType1 "Wood" :SurfaceType2 "Stone" :SurfaceType3 "Metal" :SurfaceType4 "None" :SurfaceType5 "None" :SurfaceType6 "None" :SurfaceType7 "None" :SurfaceType8 "Water" :SurfaceType9 "Grass" :SurfaceType10 "Dirt" :SurfaceType11 "None" :SurfaceType12 "None" :SurfaceType13 "None" :SurfaceType14 "Wood" :SurfaceType15 "Stone" :SurfaceType16 "Metal" :SurfaceType17 "None" :SurfaceType18 "Stone" :SurfaceType19 "None" :SurfaceType20 "None" :SurfaceType21 "None" :SurfaceType22 "None" :SurfaceType23 "Snow" :SurfaceType24 "Ice" :SurfaceType25 "Dirt" :SurfaceType26 "Sand" :SurfaceType27 "None" :SurfaceType28 "Stone" :SurfaceType29 "Stone" :SurfaceType30 "Mud" :SurfaceType31 "WetConcrete" :SurfaceType32 "TreeCanopy" :SurfaceType33 "Stone" :SurfaceType34 "None")))

(fn UpdateSurfaceSoundParameters
  (params (Physical%20Surface%20Type :type EPhysicalSurface%20Enum :default "SurfaceType_Default") (Is%20Jumping :type Boolean))
  (bind (_surface_sounds_sound_3_827e113c48bff31ece6ec89e879c345a _surface_sounds_volume_7_b64e3b9240629ed5f3cd4abea135c393 _surface_sounds_pitch_9_bbe9c03742e4f94bfd81d2b28c6b2c67) (Utilities|Struct|BreakTowHookWeaponSurfaceAudioStruct (Variables|GetSurface%20Sounds_Sound_3_827E113C48BFF31ECE6EC89E879C345A)))
  (Variables|Default|GetSkidding%20Audio
    (:then
      (bind _returnvalue (Utilities|GetDataTableRow "/TowHookWeapon/Assets/Audio/TowHookWeapon_SurfaceAudioDataTable.TowHookWeapon_SurfaceAudioDataTable" (Utilities|String|SelectName "Air" (CallFunction|GetPhysicalSurfaceName (Variables|GetPhysical%20Surface%20Type)) (Variables|GetIs%20Jumping)))
        (:then
          (LocalVariables|SetSurface%20Sounds _returnvalue))
        (:RowNotFound
          (LocalVariables|SetSurface%20Sounds (Utilities|Struct|MakeTowHookWeaponSurfaceAudioStruct (Utilities|Array|MakeArray 0)))))
      (Audio|Parameter|SetObjectParameter (Variables|Default|GetSkidding%20Audio) "Surfaces.Sound" (select (Utilities|Array|IsNotEmpty _surface_sounds_sound_3_827e113c48bff31ece6ec89e879c345a) (Utilities|Array|RandomArrayItem _surface_sounds_sound_3_827e113c48bff31ece6ec89e879c345a) 0))
      (Audio|Parameter|SetFloatParameter (Variables|Default|GetSkidding%20Audio) "Surfaces.Volume" _surface_sounds_volume_7_b64e3b9240629ed5f3cd4abea135c393 :!class /Script/AudioExtensions.AudioParameterControllerInterface)
      (Audio|Parameter|SetFloatParameter (Variables|Default|GetSkidding%20Audio) "Surfaces.Pitch" _surface_sounds_pitch_9_bbe9c03742e4f94bfd81d2b28c6b2c67 :!class /Script/AudioExtensions.AudioParameterControllerInterface)
      (Audio|Parameter|ExecuteTriggerParameter (Variables|Default|GetSkidding%20Audio) "Events.OnSurfaceChanged"))
    (:else)))

(fn ConstructionScript
  (|CallParentFunction))
