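; Looping gameplay cue for the TowHookWeapon speedlines effect that plays a tethered camera shake on the target player's camera.
; On loop start it caches the player controller and runs a repeating timer that polls movement velocity.
; The shake starts when signed X or Y velocity exceeds a threshold and none is active; otherwise it stops, so sustained speed toggles it each tick.
; OnCeaseRelevant stops the shake but leaves the timer uncleared.
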(event EventOnLoopingStart(EitherVFX-preferthisversion)
  (params (MyTarget :type Actor%20Object%20Reference) (Parameters :type Gameplay%20Cue%20Parameters%20Structure%20%28by%20ref%29) (ParticleComponents :type Array%20of%20FXSystem%20Component%20Object%20References) (AudioComponents :type Array%20of%20Audio%20Component%20Object%20References))
  (bind _asfort_player_pawn (Utilities|Casting|CastToFortPlayerPawn MyTarget))
  (Utilities|IsValid _asfort_player_pawn
    (:Is%20Valid
      (Utilities|IsValid (Fortnite|GetFortPlayerControllerfromActor _asfort_player_pawn)
        (:Is%20Valid
          (bind _output_get (Variables|Default|SetFortPlayerController (Fortnite|GetFortPlayerControllerfromActor _asfort_player_pawn)))
          (Utilities|IsValid (Variables|/Script/Engine.PlayerController|GetPlayerCameraManager _output_get)
            (:Is%20Valid
              (Utilities|Time|SetTimerbyEvent (EventDispatchers|CreateEvent :!event Custom|CheckVelocityTimer) 0.2 true true))))))))

(event Custom|CheckVelocityTimer
  (bind _fortplayercontroller (Variables|Default|GetFortPlayerController))
  (bind _playercameramanager (Variables|/Script/Engine.PlayerController|GetPlayerCameraManager _fortplayercontroller))
  (if (and (or (> (.x (Pawn|Components|CharacterMovement|GetLastUpdateVelocity (Variables|/Script/Engine.Character|GetCharacterMovement (stale EventOnLoopingStart(EitherVFX-preferthisversion) _asfort_player_pawn)))) 1e+02) (> (.y (Pawn|Components|CharacterMovement|GetLastUpdateVelocity (Variables|/Script/Engine.Character|GetCharacterMovement (stale EventOnLoopingStart(EitherVFX-preferthisversion) _asfort_player_pawn)))) 1e+02)) (not (Variables|Default|GetisShakeActive)))
    (Variables|Default|SetisShakeActive true)
    (bind _returnvalue (CameraShakes|StartCameraShake :self _playercameramanager :ShakeClass "/TowHookWeapon/Gameplay/Camera/CameraShake_TowHookTethered_Loop.CameraShake_TowHookTethered_Loop_C" :Scale (Variables|Default|GetScale) :PlaySpace "CameraLocal"))
    (Variables|Default|SetCameraShakeInstance _returnvalue)
    (else
      (Variables|Default|SetisShakeActive false)
      (CameraShakes|StopCameraShake _playercameramanager (Variables|Default|GetCameraShakeInstance) false))))

(fn OnCeaseRelevant
  (params (MyTarget :type Actor%20Object%20Reference) (Parameters :type Gameplay%20Cue%20Parameters%20Structure%20%28by%20ref%29))
  (bind _fortplayercontroller (Variables|Default|GetFortPlayerController))
  (bind _playercameramanager (Variables|/Script/Engine.PlayerController|GetPlayerCameraManager _fortplayercontroller))
  (Utilities|IsValid _fortplayercontroller
    (:Is%20Valid
      (Utilities|IsValid _playercameramanager
        (:Is%20Valid
          (CameraShakes|StopCameraShake _playercameramanager (Variables|Default|GetCameraShakeInstance) false)
          (return false))))))

(fn ConstructionScript)
